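# Control Rig for the GrindRex tail accessory cosmetic, applying a preset rotation adjustment to its tail bones.
# The execution entry feeds a single Modify Transforms unit at full weight, with fixed per-bone values starting at the tail base.
# The hierarchy holds a root, a nine-joint tail chain and pose, curl and transition curves, none of which connect to the graph.
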ControlRigBlueprint /BRCosmetics/Accessories/FORT_Tails/Tail_M_MED_GrindRex/Meshes/M_MED_GrindRex_Tail_CtrlRig
graphs [
 RigVMFunctionLibrary {links [], nodes []}
 RigVMModel {nodes [
   RigUnit_BeginExecution = &RigUnit_BeginExecution()
   &RigUnit_ModifyTransforms(
    ItemToModify="((Item=(Type=Bone,Name=\"C_Tail_A_Base_Jnt\"),Transform=(Rotation=(X=0,Y=0,Z=0.3827,W=0.9239),Translation=(X=0,Y=0,Z=0),Scale3D=(X=1,Y=1,Z=1))),(Item=(Type=Bone,Name=\"C_Tail_A_1_Jnt\"),Transform=(Rotation=(X=0,Y=-0.1305,Z=0,W=0.9914),Translation=(X=0,Y=0,Z=0) ...(truncated)"
    Weight=1
    WeightMinimum=0
    ExecutePin=@RigUnit_BeginExecution.ExecutePin
   )
  ]}
]
hierarchy [
 C_Root_Main_Root_Jnt {loc=<0 0 0>, rot=<0 0 0 1>}
 C_Tail_A_Base_Jnt {parent=C_Root_Main_Root_Jnt, loc=<1.4997 -0.3348 0.4223>, rot=<-0.5 0.5 0.5 -0.5>}
 C_Tail_A_1_Jnt {parent=C_Tail_A_Base_Jnt, loc=<8.8758 -0.0429 0.1469>, rot=<0.7071 0.0063 0.0115 -0.707>}
 C_Tail_A_2_Jnt {parent=C_Tail_A_1_Jnt, loc=<8.8775 0.0376 0.011>, rot=<0 0.0009 -0.0031 -1>}
 C_Tail_A_3_Jnt {parent=C_Tail_A_2_Jnt, loc=<8.879 0.0015 0.0004>, rot=<0 0 -0.0001 -1>}
 C_Tail_A_4_Jnt {parent=C_Tail_A_3_Jnt, loc=<8.8778 0 0>, rot=<0 0 0 1>}
 C_Tail_A_5_Jnt {parent=C_Tail_A_4_Jnt, loc=<8.8778 0 0>, rot=<0 0 0 1>}
 C_Tail_A_6_Jnt {parent=C_Tail_A_5_Jnt, loc=<8.8775 0 0>, rot=<0 0 0 1>}
 C_Tail_A_7_Jnt {parent=C_Tail_A_6_Jnt, loc=<8.8776 0 0>, rot=<0 0 0 1>}
 C_Tail_A_8_Jnt {parent=C_Tail_A_7_Jnt, loc=<8.8788 0 0>, rot=<0 0 0 1>}
 Pose_0 {type=Curve, loc=<0 0 0>, rot=<0 0 0 1>}
 Pose_1 {type=Curve, loc=<0 0 0>, rot=<0 0 0 1>}
 Pose_2 {type=Curve, loc=<0 0 0>, rot=<0 0 0 1>}
 curl_up {type=Curve, loc=<0 0 0>, rot=<0 0 0 1>}
 curl_down {type=Curve, loc=<0 0 0>, rot=<0 0 0 1>}
 Transition {type=Curve, loc=<0 0 0>, rot=<0 0 0 1>}
]
